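; CTD_OnFiringRangedWeaponByRarity is a cosmetic trigger asset for the Athena cosmetic modifier system, named for firing ranged weapons by rarity.
; It overrides GetTriggerInstanceClass to return the CT_OnFiringRangedWeaponByRarity Blueprint class as its trigger instance.
; The file holds no firing or rarity logic of its own, only this link to the instance class.
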(fn GetTriggerInstanceClass
  (return "/BRCosmetics/Athena/Cosmetics/CosmeticModifiers/CosmeticTriggers/CT_OnFiringRangedWeaponByRarity.CT_OnFiringRangedWeaponByRarity_C"))
