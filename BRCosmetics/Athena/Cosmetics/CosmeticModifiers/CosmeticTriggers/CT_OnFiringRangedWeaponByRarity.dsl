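; Cosmetic trigger that fires when the instigating player fires a ranged weapon, optionally limited to certain weapon rarities.
; On initialize it caches the context's player pawn and its CTD_OnFiringRangedWeaponByRarity definition.
; It binds the pawn's OnFiringRangedWeapon event on activate and unbinds it on deactivate.
; Each shot requests the trigger if the definition's rarity set is empty or contains the ability's source weapon rarity.
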(event EventOnInitialize
  (params (InCosmeticTriggerDefinition :type Cosmetic%20Trigger%20Definition%20Object%20Reference) (InTriggerContext :type Cosmetic%20Trigger%20Context%20Structure%20%28by%20ref%29))
  (|CallParentFunction InCosmeticTriggerDefinition InTriggerContext)
  (Utilities|FlowControl|Sequence
    (:then_0
      ; Cache the player instigator
      (bind _asfort_player_pawn (Utilities|Casting|CastToFortPlayerPawn (Utilities|Struct|BreakCosmeticTriggerContext (Variables|CosmeticTrigger|GetContext))))
      (Variables|Default|SetTriggerInstigator _asfort_player_pawn))
    (:then_1
      ; Resolve the trigger definition that contains the data for this trigger instance
      (bind _asctd__onfiringrangedweaponbyrarity (Utilities|Casting|CastToCTD_OnFiringRangedWeaponByRarity (Variables|CosmeticTrigger|GetDefinition)))
      (Variables|Default|SetTriggerDefinition _asctd__onfiringrangedweaponbyrarity))))

(event EventOnActivate
  (|CallParentFunction)
  ; Bind to Gameplay Events
  (bind _triggerinstigator (Variables|Default|GetTriggerInstigator)
    (:then
      (EventDispatchers|BindEventtoOnFiringRangedWeapon _triggerinstigator (EventDispatchers|CreateEvent :!event Custom|OnFiringRangedWeapon)))
    (:else)))

(event EventOnDeactivate
  (bind _triggerinstigator (Variables|Default|GetTriggerInstigator))
  (|CallParentFunction)
  ; Unbind From gameplay events
  (Utilities|IsValid _triggerinstigator
    (:Is%20Valid
      (EventDispatchers|UnbindEventfromOnFiringRangedWeapon _triggerinstigator (EventDispatchers|CreateEvent :!event Custom|OnFiringRangedWeapon)))))

(event Custom|OnFiringRangedWeapon
  (params (GameplayAbility :type Fort%20Gameplay%20Ability%20Object%20Reference) (AbilityTags :type Gameplay%20Tag%20Container%20Structure))
  (if (or (Utilities|Set|IsEmpty (Variables|/BRCosmetics/Athena/Cosmetics/CosmeticModifiers/CosmeticTriggers/CTD_OnFiringRangedWeaponByRarity.CTD_OnFiringRangedWeaponByRarity_C|GetRestrictByWeaponRarity (Variables|Default|GetTriggerDefinition))) (Utilities|Set|ContainsItem (Variables|/BRCosmetics/Athena/Cosmetics/CosmeticModifiers/CosmeticTriggers/CTD_OnFiringRangedWeaponByRarity.CTD_OnFiringRangedWeaponByRarity_C|GetRestrictByWeaponRarity (Variables|Default|GetTriggerDefinition)) (Rarity|GetWeaponDataRarity (Ability|GetCurrentSourceWeapon GameplayAbility))))
    (CallFunction|RequestFireTrigger (Utilities|Struct|MakeCosmeticEventPayload))))
